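# Unlit, two-sided translucent material for Niagara mesh particles in the DaffodilSong pickaxe blade effects, giving an animated orange-gold glow.
# Hand-built polar UVs pan blurred noise radially, with explicit derivatives from unwrapped UVs for mip selection.
# Vertex red masks shape noise for opacity, with particle alpha and depth fade, and for a two-color blend with optional sphere gradient.
# A red-masked HDR orange boost feeds into emissive adjustments.
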Material /BRCosmetics/Weapons/FORT_Melee/Pickaxe_DaffodilSong/FX/Material/M_Pickaxe_DaffodilSong
graph {
 properties {
  bUsedWithNiagaraMeshParticles=true
  blendMode=BLEND_Translucent
  pixelDepthOffsetMode=PDOM_Legacy
  previewMesh=/BRCosmetics/Weapons/FORT_Melee/Pickaxe_DaffodilSong/FX/Mesh/SM_Pickaxe_DaffodilSong_FX_Blade
  shadingModel=MSM_Unlit
  twoSided=true
  usedShadingModels=Unlit
 }
 nodes [
  n1 = ScalarParameter() {defaultValue=6, parameterName=s}
  Add(A=@n1, B=ParticleRandom())
  MaterialFunctionCall() {materialFunction=&DebugTimeSine}
  n2 = ScalarParameter() {defaultValue=0.3, parameterName=Param_2}
  n3 = VertexColor()
  n4 = OneMinus(None=@n3.R)
  n5 = TextureCoordinate()
  n6 = ScalarParameter() {defaultValue=0.76, parameterName=X}
  n7 = ScalarParameter() {defaultValue=1, parameterName=Y}
  n8 = Subtract(A=Multiply(A=@n5, B=AppendVector(A=@n6, B=@n7))) {constB=0.5}
  n9 = Arctangent2Fast(Y=ComponentMask(None=@n8) {r=true}, X=ComponentMask(None=@n8) {g=true})
  n10 = ScalarParameter() {defaultValue=6}
  n11 = AppendVector(A=Length(None=@n8) {is=class MaterialExpressionLength}, B=Frac(None=Divide(A=@n9, B=@n10)))
  n12 = Constant2Vector() {g=10, r=1}
  n13 = DDX(Value=@n5)
  n14 = DDY(Value=@n5)
  n15 = TextureSample!(UVs=Panner(Coordinate=Multiply(A=@n11, B=@n12)) {speedX=-0.5}, Tex, "DDX(UVs)"=@n13, "DDY(UVs)"=@n14, Apply View MipBias) {mipValueMode=TMVM_Derivative, samplerType=SAMPLERTYPE_Grayscale, texture=/FortWater/Textures/LowResBlurredNoise}
  n16 = OneMinus(None=@n3.R)
  n17 = ScalarParameter() {defaultValue=0.8, parameterName=X2}
  n18 = ScalarParameter() {defaultValue=1, parameterName=Y2}
  n19 = Subtract(A=Multiply(A=TextureCoordinate(), B=AppendVector(A=@n17, B=@n18))) {constB=0.5}
  n20 = Arctangent2Fast(Y=ComponentMask(None=@n19) {r=true}, X=ComponentMask(None=@n19) {g=true})
  n21 = AppendVector(A=Length(None=@n19), B=Frac(None=Divide(A=@n20, B=@n1)))
  n22 = Constant2Vector() {g=13, r=1}
  n23 = TextureSample(UVs=Panner(Coordinate=Multiply(A=@n21, B=@n22)) {speedX=-1}, "DDX(UVs)"=@n13, "DDY(UVs)"=@n14) {mipValueMode=TMVM_Derivative, samplerType=SAMPLERTYPE_Grayscale, texture=&LowResBlurredNoise}
  n24 = Add(
   A=Add(A=SmoothStep(Min=@n2, Value=@n4) {constMax=0.5, constMin=0.3}, B=SmoothStep(Value=@n15.RGB) {constMax=0.7, constMin=0.2})
   B=Multiply(A=SmoothStep(Value=@n16) {constMax=0.5, constMin=0.1}, B=SmoothStep(Value=@n23.RGB) {constMax=0.7, constMin=0.2})
  )
  n25 = SmoothStep(Value=OneMinus(None=@n3.R)) {constMax=0.3, constMin=0.2}
  n26 = Multiply(A=Multiply(A=Saturate(None=@n24), B=Saturate(None=@n25))) {constB=2}
  Multiply(A=ParticleColor().A, B=DepthFade(Opacity=Saturate(None=@n26), FadeDistance=ScalarParameter() {defaultValue=3, parameterName=Param_3}))
  n27 = SmoothStep(Value=OneMinus(None=@n3.R)) {constMin=0.25}
  n28 = ScalarParameter() {defaultValue=0.76, parameterName=x3}
  n29 = ScalarParameter() {defaultValue=1, parameterName=y3}
  n30 = Subtract(A=Multiply(A=TextureCoordinate() {uTiling=0.76}, B=AppendVector(A=@n28, B=@n29))) {constB=0.5}
  n31 = Divide(A=Arctangent2Fast(Y=ComponentMask(None=@n30) {r=true}, X=ComponentMask(None=@n30) {g=true}), B=Constant() {r=4})
  n32 = Multiply(A=AppendVector(A=Length(None=@n30), B=Frac(None=@n31)), B=Constant2Vector() {g=10, r=1})
  n33 = TextureCoordinate() {uTiling=0.76}
  n34 = DDX(Value=@n33)
  n35 = DDY(Value=@n33)
  n36 = SmoothStep(Value=TextureSample(UVs=Panner(Coordinate=@n32) {speedX=-1}, "DDX(UVs)"=@n34, "DDY(UVs)"=@n35) {mipValueMode=TMVM_Derivative, samplerType=SAMPLERTYPE_Grayscale, texture=&LowResBlurredNoise}.RGB) {constMax=0.7, constMin=0.2}
  n37 = Subtract(A=@n33) {constB=0.5}
  n38 = Length(None=@n37)
  n39 = ComponentMask(None=@n37) {r=true}
  n40 = ComponentMask(None=@n37) {g=true}
  n41 = Frac(None=Divide(A=Arctangent2Fast(Y=@n39, X=@n40), B=Constant() {r=66}))
  n42 = Panner(Coordinate=Multiply(A=AppendVector(A=@n38, B=@n41), B=Constant2Vector() {g=6, r=1})) {speedX=-0.5}
  n43 = SmoothStep(Value=Power(Base=TextureSample(UVs=@n42, "DDX(UVs)"=@n34, "DDY(UVs)"=@n35) {mipValueMode=TMVM_Derivative, samplerType=SAMPLERTYPE_Grayscale, texture=&LowResBlurredNoise}.RGB) {constExponent=0.4}) {constMax=0.4}
  n44 = SmoothStep(
   Min=ScalarParameter() {defaultValue=0.55, parameterName=ColorMask_Start}
   Max=ScalarParameter() {defaultValue=0.8, parameterName=ColorMask_End}
   Value=OneMinus(None=@n3.R)
  ) {constMax=0.8, constMin=0.55}
  n45 = Saturate(None=Add(A=Multiply(A=@n27, B=@n36), B=Multiply(A=@n43, B=@n44)))
  n46 = TextureCoordinate() {uTiling=0.76}
  n47 = Constant() {r=0.35}
  n48 = Add(
   A=Power(Base=MaterialFunctionCall!(UVs=@n46, CenterPosition, Radius=@n47) {materialFunction=/Engine/Functions/Engine_MaterialFunctions01/Gradient/SphereGradient-2D}.Result 0-1) {constExponent=9}
   B=@n45
  )
  n49 = ScalarParameter() {defaultValue=1, parameterName=UseSphereMask?}
  n50 = LinearInterpolate(
   A=VectorParameter() {defaultValue=<rgba 1 0.5909 0.0472 1>}.RGB
   B=VectorParameter() {parameterName=Param_1, defaultValue=<rgba 2 1.2781 0.5423 1>}.RGB
   Alpha=Saturate(None=LinearInterpolate(A=@n45, B=@n48, Alpha=@n49))
  )
  n51 = Multiply(
   A=SmoothStep(Min=ScalarParameter() {defaultValue=0.6, parameterName=Param_1}, Value=@n3.R) {constMin=0.6}
   B=Constant3Vector() {constant=<rgba 22 3.7136 0 1>}
  )
  Reroute(None=MaterialFunctionCall(Emissive=Add(A=@n50, B=@n51)) {materialFunction=&MF_FX_EmissiveAdjustments}.Result)
  ScalarParameter() {defaultValue=0.0731, parameterName=Param_4}
 ]
}
